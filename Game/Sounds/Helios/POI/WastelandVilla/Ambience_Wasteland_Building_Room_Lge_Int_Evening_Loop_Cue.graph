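# Looping evening ambience cue for a large interior room at the Wasteland Villa POI, blending stone room tone with desert night.
# A quality switch gives the first two levels a quad branch, mixing one of three random room tones with the desert bed.
# The last two levels get a lighter stereo branch with one room tone and randomized pitch and volume.
# Looping envelopes slowly vary pitch and volume on both branches.
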SoundCue /Game/Sounds/Helios/POI/WastelandVilla/Ambience_Wasteland_Building_Room_Lge_Int_Evening_Loop_Cue
root=QualityLevel_0
nodes [
 {
  id=Enveloper_0
  children [Enveloper_3]
  properties {
   bLoop=true
   bLoopIndefinitely=true
   loopEnd=100
   pitchCurve {editorCurveData {keys [<at 0 0.8>]}}
   volumeCurve {editorCurveData {keys [<at 0 1 RCIM_Cubic>,<at 50 0.9 RCIM_Cubic>,<at 100 1 RCIM_Cubic>]}}
  }
 }
 {
  id=Enveloper_1
  children [Mixer_1]
  properties {
   bLoop=true
   bLoopIndefinitely=true
   loopEnd=30
   pitchCurve {editorCurveData {keys [{interpMode=RCIM_Cubic, time=0}, <at 15.9468 0.9181 RCIM_Cubic>, {interpMode=RCIM_Cubic, time=30}]}}
   volumeCurve {editorCurveData {keys [<at 0 1 RCIM_Cubic>,<at 16.1421 0.921 RCIM_Cubic>,<at 30 1 RCIM_Cubic>]}}
  }
 }
 {
  id=Enveloper_3
  children [Mixer_0]
  properties {
   bLoop=true
   bLoopIndefinitely=true
   loopEnd=30
   pitchCurve {editorCurveData {keys [{interpMode=RCIM_Cubic, time=0}, <at 15.9468 0.9181 RCIM_Cubic>, {interpMode=RCIM_Cubic, time=30}]}}
   volumeCurve {editorCurveData {keys [<at 0 1>]}}
  }
 }
 {id=Mixer_0, children [Modulator_11,WavePlayer_5], properties {inputVolume [0.5,0.35]}}
 {id=Mixer_1, children [WavePlayer_12, WavePlayer_11], properties {inputVolume [0.5,0.35]}}
 {id=Modulator_11, children [Random_0], properties {pitchMax=1.1, pitchMin=1.1}}
 {id=Modulator_2, children [Enveloper_1], properties {pitchMax=1.1, pitchMin=0.9, volumeMax=1.1, volumeMin=0.9}}
 {id=QualityLevel_0, children [Enveloper_0,Enveloper_0,Modulator_2,Modulator_2]}
 {id=Random_0, children [WavePlayer_6, WavePlayer_7, WavePlayer_10], properties {weights [1,1,1]}}
 {
  id=WavePlayer_10
  properties {bLooping=true, soundWaveAssetPtr=/Game/Sounds/Artemis/RoomTones/Stone/Artemis_Interior_Large_Stone_Room_03_Quad}
 }
 {id=WavePlayer_11, properties {bLooping=true, soundWaveAssetPtr=/Game/Sounds/Artemis/Biomes/Desert/Ambience_Desert_Night_Stereo}}
 {
  id=WavePlayer_12
  properties {bLooping=true, soundWaveAssetPtr=/Game/Sounds/Artemis/RoomTones/Stone/Artemis_Interior_Large_Stone_Room_01_Stereo}
 }
 {id=WavePlayer_5, properties {bLooping=true, soundWaveAssetPtr=/Game/Sounds/Artemis/Biomes/Desert/Ambience_Desert_Night_Quad}}
 {
  id=WavePlayer_6
  properties {bLooping=true, soundWaveAssetPtr=/Game/Sounds/Artemis/RoomTones/Stone/Artemis_Interior_Large_Stone_Room_01_Quad}
 }
 {
  id=WavePlayer_7
  properties {bLooping=true, soundWaveAssetPtr=/Game/Sounds/Artemis/RoomTones/Stone/Artemis_Interior_Large_Stone_Room_02_Quad}
 }
]
properties {maxDistance=2097152, soundClassObject=&Fort_Ambient_Indoors_Athena, volumeMultiplier=0.501}
