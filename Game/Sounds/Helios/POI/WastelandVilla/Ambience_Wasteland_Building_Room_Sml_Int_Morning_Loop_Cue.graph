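# Looping indoor morning ambience cue for a small building room at the Wasteland Villa POI, in the indoor ambient sound class.
# A quality switch feeds its first two levels a quad branch and its last two a stereo branch, both mixing stone room tone with desert daytime ambience.
# Quad randomly picks one of three room tones; stereo uses one, with randomized pitch and volume.
# Endlessly looping envelopes shape pitch and volume over time.
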SoundCue /Game/Sounds/Helios/POI/WastelandVilla/Ambience_Wasteland_Building_Room_Sml_Int_Morning_Loop_Cue
root=QualityLevel_0
nodes [
 {
  id=Enveloper_0
  children [Enveloper_3]
  properties {
   bLoop=true
   bLoopIndefinitely=true
   loopEnd=100
   pitchCurve {editorCurveData {keys [<at 0 0.8>]}}
   volumeCurve {editorCurveData {keys [<at 0 1 RCIM_Cubic>,<at 50 0.9 RCIM_Cubic>,<at 100 1 RCIM_Cubic>]}}
  }
 }
 {
  id=Enveloper_1
  children [Mixer_1]
  properties {
   bLoop=true
   bLoopIndefinitely=true
   loopEnd=30
   pitchCurve {editorCurveData {keys [{interpMode=RCIM_Cubic, time=0}, <at 15.9468 0.9181 RCIM_Cubic>, {interpMode=RCIM_Cubic, time=30}]}}
   volumeCurve {editorCurveData {keys [<at 0 1 RCIM_Cubic>,<at 16.1421 0.921 RCIM_Cubic>,<at 30 1 RCIM_Cubic>]}}
  }
 }
 {
  id=Enveloper_3
  children [Mixer_0]
  properties {
   bLoop=true
   bLoopIndefinitely=true
   loopEnd=30
   pitchCurve {editorCurveData {keys [{interpMode=RCIM_Cubic, time=0}, <at 15.9468 0.9181 RCIM_Cubic>, {interpMode=RCIM_Cubic, time=30}]}}
   volumeCurve {editorCurveData {keys [<at 0 1>]}}
  }
 }
 {id=Mixer_0, children [Modulator_11,WavePlayer_5], properties {inputVolume [0.5,0.25]}}
 {id=Mixer_1, children [WavePlayer_8, WavePlayer_11], properties {inputVolume [0.5,0.35]}}
 {id=Modulator_11, children [Random_0], properties {pitchMax=1.1, pitchMin=1.1}}
 {id=Modulator_2, children [Enveloper_1], properties {pitchMax=1.1, pitchMin=0.9, volumeMax=1.1, volumeMin=0.9}}
 {id=QualityLevel_0, children [Enveloper_0,Enveloper_0,Modulator_2,Modulator_2]}
 {id=Random_0, children [WavePlayer_0,WavePlayer_2,WavePlayer_4], properties {weights [1,1,1]}}
 {
  id=WavePlayer_0
  properties {bLooping=true, soundWaveAssetPtr=/Game/Sounds/Artemis/RoomTones/Stone/Artemis_Interior_Medium_Stone_Room_01_Quad}
 }
 {id=WavePlayer_11, properties {bLooping=true, soundWaveAssetPtr=/Game/Sounds/Artemis/Biomes/Desert/Ambience_Desert_Day_Stereo}}
 {
  id=WavePlayer_2
  properties {bLooping=true, soundWaveAssetPtr=/Game/Sounds/Artemis/RoomTones/Stone/Artemis_Interior_Medium_Stone_Room_02_Quad}
 }
 {
  id=WavePlayer_4
  properties {bLooping=true, soundWaveAssetPtr=/Game/Sounds/Artemis/RoomTones/Stone/Artemis_Interior_Small_Stone_Room_01_Quad}
 }
 {id=WavePlayer_5, properties {bLooping=true, soundWaveAssetPtr=/Game/Sounds/Artemis/Biomes/Desert/Ambience_Desert_Day_Quad}}
 {
  id=WavePlayer_8
  properties {bLooping=true, soundWaveAssetPtr=/Game/Sounds/Artemis/RoomTones/Stone/Artemis_Interior_Small_Stone_Room_01_Stereo}
 }
]
properties {maxDistance=2097152, soundClassObject=&Fort_Ambient_Indoors_Athena, volumeMultiplier=0.376}
